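# Masked unlit material for the Missus Mind pickaxe trail, usable on Niagara ribbons and sprites.
# Panning noise blends with the Mastermind twisted normal map to fake lit highlights and rim from a reflection vector against a fixed or sky-atmosphere light.
# A noise-distorted sine band over UVs, plus particle alpha, shapes diffuse and edge color.
# Dynamic parameters drive offsets, time-of-day tint and light intensities; VariantLerp fades to a flat grey variant.
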Material /BRCosmetics/Weapons/FORT_Melee/Pickaxe_MissusMind/FX/M_MissusMind_Pickaxe_Trail
graph {
 properties {
  bAutomaticallySetUsageInEditor=false
  bGenerateSphericalParticleNormals=true
  bUsedWithNiagaraRibbons=true
  bUsedWithNiagaraSprites=true
  blendMode=BLEND_Masked
  shadingModel=MSM_Unlit
  translucencyPass=MTP_BeforeDOF
  usedShadingModels=Unlit
 }
 nodes [
  n1 = DynamicParameter() {paramNames [Offset,Offset2,DiffuseNight,TODMulti]}
  n2 = Multiply(
   A=Add(A=TextureCoordinate() {coordinateIndex=1}, B=AppendVector(A=@n1.Offset2, B=@n1.Offset2))
   B=AppendVector(A=ScalarParameter() {defaultValue=0.8, parameterName=UTile}, B=ScalarParameter() {defaultValue=0.2, parameterName=VTile})
  )
  n3 = TextureSample(UVs=Panner(Coordinate=Multiply(A=@n2) {constB=2}) {speedY=0.1}) {
   samplerType=SAMPLERTYPE_LinearColor
   texture=/BRCosmetics/Accessories/FORT_Backpacks/Backpack_M_MED_CrystalGlobe/FX/Textures/T_Backpack_CrystalGlobe_Noise
  }
  n4 = ConstantBiasScale(None=@n3.RGB) {bias=-0.5, scale=2}
  n5 = LinearInterpolate(B=Constant() {r=0.6}, Alpha=ScalarParameter() {parameterName=VariantLerp}) {constA=0.2}
  n6 = TextureCoordinate() {coordinateIndex=1}
  n7 = AppendVector(A=@n1.Offset, B=@n1.Offset)
  n8 = TextureSample(UVs=Multiply(A=Add(A=@n6, B=@n7), B=Constant2Vector() {g=0.9, r=1.6})) {samplerType=SAMPLERTYPE_Normal, texture=/Game/Characters/Player/Male/Medium/Bodies/M_MED_Mastermind/Textures/T_M_MED_Mastermind_Twisted_N}
  n9 = MaterialFunctionCall(
   Normal=MaterialFunctionCall(
    BaseNormal=MaterialFunctionCall(Normal=@n4, Flatness=@n5) {materialFunction=&FlattenNormal}.Result
    AdditionalNormal=@n8.RGBA
   ) {materialFunction=&BlendAngleCorrectedNormals}.Result
   Flatness=Constant() {r=0.2}
  ) {materialFunction=&FlattenNormal}
  n10 = DynamicParameter() {parameterIndex=1, paramNames [VariantEdgeMult, FELerp, Param3, Param4]}
  Clamp(None=@n9.Result, Max=LinearInterpolate(Alpha=@n10.FELerp) {constA=1, constB=0.65}) {minDefault=-1}
  Constant2Vector() {g=0.35, r=1}
  Constant3Vector() {constant=<rgba 0.0151 0.0272 0.0525 1>}
  Constant() {r=0.328}
  Constant() {r=0.5}
  Constant() {r=1}
  n11 = ParticleColor()
  n12 = Multiply(A=@n3.A) {constB=0.65}
  n13 = Add(A=TextureCoordinate(), B=AppendVector(A=Constant(), B=ScalarParameter() {defaultValue=-0.244, parameterName=UOffset}))
  n14 = Sine(None=ComponentMask(None=Add(A=@n12, B=@n13)) {g=true}) {period=2}
  n15 = ComponentMask(None=TextureCoordinate()) {r=true}
  n16 = Multiply(A=Saturate(None=Power(Base=@n14) {constExponent=7}), B=Power(Base=@n15) {constExponent=0.2})
  n17 = Power(Base=OneMinus(None=@n15)) {constExponent=0.1}
  n18 = Add(A=Multiply(A=Multiply(A=@n16, B=@n17), B=ScalarParameter() {defaultValue=1, parameterName=OpacityMult}), B=@n11.A)
  n19 = Saturate(None=@n18)
  n20 = Subtract(A=@n11.RGB, B=Multiply(A=OneMinus(None=@n19)) {constB=0.4}) {constA=0, constB=0}
  n21 = DynamicParameter() {parameterIndex=1, paramNames [VariantEdgeMult, FELerp, Param3, Param4]}
  n22 = LinearInterpolate(Alpha=@n10.FELerp) {constA=1.8, constB=0.8}
  n23 = Saturate(None=Multiply(A=Power(Base=@n18, Exp=@n22) {constExponent=0.8}))
  LinearInterpolate(A=Multiply(A=@n20, B=@n21.VariantEdgeMult) {constB=0.9}, B=@n20, Alpha=@n23)
  LinearInterpolate(B=Constant() {r=0.25}, Alpha=DynamicParameter() {parameterIndex=1, paramNames [VariantEdgeMult, FELerp, Param3, Param4]}.FELerp) {constA=0.4}
  n24 = Constant3Vector() {constant=<rgba 0 0 -1 0>}
  n25 = ReflectionVectorWS(CustomWorldNormal=@n9.Result)
  n26 = DotProduct(A=LinearInterpolate(A=@n24, B=SkyAtmosphereLightDirection(), Alpha=@n21.FELerp) {constA=1}, B=@n25)
  n27 = ConstantBiasScale(None=@n26)
  n28 = DynamicParameter() {parameterIndex=3, paramNames [HighlightR,HighlightG,HighlightB,Param4]}
  n29 = MaterialFunctionCall(R=@n28.HighlightR, G=@n28.HighlightR, B=@n28.HighlightR) {materialFunction=&AppendMany}
  n30 = Saturate(None=@n26)
  n31 = SkyAtmosphereDistantLightScatteredLuminance()
  n32 = DynamicParameter() {paramNames [Offset,Offset2,DiffuseNight,TODMulti]}
  n33 = LinearInterpolate(
   A=VectorParameter() {parameterName=Diffuse, defaultValue=<rgba 0.015 0.015 0.015 0>}.RGB
   B=Constant3Vector() {constant=<rgba 0.0052 0.0068 0.015 1>}
   Alpha=@n32.DiffuseNight
  )
  n34 = Add(A=Multiply(A=Constant() {r=0.05}, B=Multiply(A=@n30, B=@n31)), B=Multiply(A=@n18, B=Multiply(A=@n33, B=@n32.TODMulti)))
  n35 = SkyAtmosphereLightDirection()
  n36 = DotProduct(A=OneMinus(None=LinearInterpolate(A=@n24, B=@n35, Alpha=@n21.FELerp) {constA=1}), B=@n25)
  n37 = DynamicParameter() {parameterIndex=2, defaultValue=<rgba 1 1 1 0.9>, paramNames [RimLightR,RimLightG,RimLightB,Param4]}
  n38 = Multiply(
   A=Multiply(A=Power(Base=@n36) {constExponent=4}) {constB=0.1}
   B=MaterialFunctionCall(R=@n37.RimLightR, G=@n37.RimLightR, B=@n37.RimLightR) {materialFunction=&AppendMany}.RGB
  )
  n39 = Power(Base=OneMinus(None=Length(None=@n18) {is=class MaterialExpressionLength})) {constExponent=4}
  n40 = Multiply(A=@n32.TODMulti, B=VectorParameter() {parameterName=Edge_Color, defaultValue=<rgba 0.025 0.025 0.025 1>}.RGB)
  n41 = Add(
   A=Multiply(A=Power(Base=@n27) {constExponent=24}, B=Multiply(A=@n29.RGB) {constB=0.36})
   B=Add(A=Add(A=@n34, B=@n38), B=Multiply(A=@n39, B=@n40))
  )
  n42 = Multiply(A=Constant() {r=0.2}, B=@n32.TODMulti)
  n43 = Multiply(A=@n42) {constB=0.7}
  MaterialFunctionCall(Emissive=LinearInterpolate(A=Multiply(A=@n41, B=@n11.RGB), B=LinearInterpolate(A=@n43, B=@n42, Alpha=@n23), Alpha=ScalarParameter() {parameterName=VariantLerp})) {materialFunction=&MF_FX_EmissiveAdjustments}
  VectorParameter() {parameterName=Specular Highlight_1, defaultValue=<rgba 0.4397 0.2582 0.1559 1>}
 ]
}
