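# Animation Blueprint for the MissusMind pickaxe that adds subtle secondary motion to its two blade bones.
# The reference pose passes through DefaultSlot, then two chained Anim Dynamics nodes simulate blade_01 and blade_02 at partial alpha.
# Both use damped linear springs with zero gravity, motion limits and scaled, clamped component movement.
# The event graph holds only unconnected stub nodes, so all behavior lives in the AnimGraph.
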AnimBlueprint /BRCosmetics/Weapons/FORT_Melee/Pickaxe_MissusMind/Meshes/MissusMind_Axe_AnimBP
target_skeleton=&/MissusMind_Axe_Skeleton
graphs [
 AnimGraph {nodes [
   n1 = Local Space Ref Pose()
   n2 = Anim Dynamics - Bone: blade_01(ComponentPose=Local To Component(LocalPose=Slot 'DefaultSlot'(Source=@n1)), Alpha=0.25, ExternalForce="0,0,0") {
    angularDampingOverride=0.9
    angularSpringConstant=200
    bGravityOverrideInSimSpace=true
    bLinearSpring=true
    bOverrideAngularDamping=true
    bOverrideLinearDamping=true
    gravityScale=0
    linearDampingOverride=0.9999
    linearSpringConstant=100
    alphaBoolBlend {blendOption=Linear}
    boundBone {boneName=blade_01}
    componentAppliedLinearAccClamp=<3000 3000 3000>
    componentLinearAccScale=<0.3 0.3 0.3>
    componentLinearVelScale=<0.3 0.3 0.3>
    physicsBodyDefinitions [{
      sphereCollisionRadius=0
      boundBone {boneName=blade_01}
      boxExtents=<4 4 4>
      constraintSetup {angularLimitsMax=<40 40 40>, angularLimitsMin=<-40 -40 -40>, angularTarget=<0 0 0>, linearAxesMax=<30 30 30>, linearAxesMin=<-30 -30 -30>}
      localJointOffset=<0 0 0>
     }]
    relativeSpaceBone {boneName=b_AR_Root}
   }
   n3 = Anim Dynamics - Bone: blade_02(ComponentPose=@n2, Alpha=0.25, ExternalForce="0,0,0") {
    angularDampingOverride=0.9
    angularSpringConstant=200
    bGravityOverrideInSimSpace=true
    bLinearSpring=true
    bOverrideAngularDamping=true
    bOverrideLinearDamping=true
    gravityScale=0
    linearDampingOverride=0.9999
    linearSpringConstant=100
    alphaBoolBlend {blendOption=Linear}
    boundBone {boneName=blade_02}
    componentAppliedLinearAccClamp=<3000 3000 3000>
    componentLinearAccScale=<0.3 0.3 0.3>
    componentLinearVelScale=<0.3 0.3 0.3>
    physicsBodyDefinitions [{
      sphereCollisionRadius=0
      boundBone {boneName=blade_02}
      boxExtents=<4 4 4>
      constraintSetup {angularLimitsMax=<40 40 40>, angularLimitsMin=<-40 -40 -40>, angularTarget=<0 0 0>, linearAxesMax=<30 30 30>, linearAxesMin=<-30 -30 -30>}
      localJointOffset=<0 0 0>
     }]
    relativeSpaceBone {boneName=b_AR_Root}
   }
   Output Pose(Result=Component To Local(ComponentPose=@n3))
  ]}
 EventGraph {nodes [
   TryGetPawnOwner()
   Event BlueprintUpdateAnimation()
  ]}
]
variables []
