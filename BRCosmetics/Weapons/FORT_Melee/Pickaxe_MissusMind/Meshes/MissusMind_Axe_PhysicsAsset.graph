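# Physics asset for the MissusMind pickaxe, a two-handed melee harvesting tool cosmetic, giving its axe mesh a simple collision shape.
# A single body on the Melee_TwoHanded_Handle bone holds one capsule, offset and rotated from the bone.
# It has no constraints, so it defines no joints between bodies.
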PhysicsAsset /BRCosmetics/Weapons/FORT_Melee/Pickaxe_MissusMind/Meshes/MissusMind_Axe_PhysicsAsset
bodies {Melee_TwoHanded_Handle [{shapeType=Capsule, length=49.5121, radius=16.4182, center=<-0.3 -13.8256 2.4276>, rotation=<rot 0 0.8931 -84.7152>}]}
constraints []
